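# Physics asset for the SparkArcher backpack cosmetic mesh, setting up secondary motion for its dangling straps and three arrows.
# Capsules cover arrows, strap bones and a pelvis joint, while three boxes form the pack body.
# Arrows hang from the pack with very tight swing and twist limits.
# A limited-swing strap chain runs from the pack to a locked bind joint, then forks into upper and lower branches.
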PhysicsAsset /BRCosmetics/Accessories/FORT_Backpacks/Backpack_F_MED_SparkArcher/Meshes/F_MED_SparkArcher_Pack_Physics
bodies {
 dyn_arrows_jnt_01 [{
   shapeType=Capsule
   length=20.1688
   radius=1.8287
   shapeName=dyn_arrows_jnt_01_capsule
   center=<-0.0627 0.4295 -0.2457>
   rotation=<rot 1.3825 -90 -0.5626>
  }]
 dyn_arrows_jnt_02 [{
   shapeType=Capsule
   length=20.0711
   radius=1.4599
   shapeName=dyn_arrows_jnt_02_capsule
   center=<0.2174 0.9236 -0.2749>
   rotation=<rot 2.4155 -90 -2.2312>
  }]
 dyn_arrows_jnt_03 [{
   shapeType=Capsule
   length=20.304
   radius=2.2278
   shapeName=dyn_arrows_jnt_03_capsule
   center=<0.5364 0.0684 0.1912>
   rotation=<rot 0.8822 -90 -0.4403>
  }]
 dyn_strap_bind_jnt_01 [{shapeType=Capsule, length=1, radius=1, shapeName=dyn_strap_bind_jnt_01_capsule, rotation=<rot 79.6518 179.9998 179.9997>}]
 dyn_strap_jnt_01 [{shapeType=Capsule, length=1.5768, radius=0.9157, shapeName=dyn_strap_jnt_01_capsule, rotation=<rot 90 0 0>}]
 dyn_strap_jnt_02 [{shapeType=Capsule, length=1.919, radius=1, shapeName=dyn_strap_jnt_02_capsule, rotation=<rot 90 0 0>}]
 dyn_strap_jnt_03 [{shapeType=Capsule, length=2.37, radius=1, shapeName=dyn_strap_jnt_03_capsule, rotation=<rot 90 0 0>}]
 dyn_strap_jnt_04 [{shapeType=Capsule, length=2.0054, radius=1, shapeName=dyn_strap_jnt_04_capsule, rotation=<rot 90 0 0>}]
 dyn_strap_jnt_05 [{shapeType=Capsule, length=1.5097, radius=1, shapeName=dyn_strap_jnt_05_capsule, rotation=<rot 82.8647 0.0006 0.0006>}]
 dyn_strap_lower_jnt_01 [{shapeType=Capsule, length=1, radius=1, shapeName=dyn_strap_lower_jnt_01_capsule, rotation=<rot 90 0 0>}]
 dyn_strap_lower_jnt_02 [{shapeType=Capsule, length=1.9283, radius=1, shapeName=dyn_strap_lower_jnt_02_capsule, center=<1.0717 -0.0565 -0.0432>, rotation=<rot 90 0 0>}]
 dyn_strap_lower_jnt_03 [{shapeType=Capsule, length=2.7825, radius=1, shapeName=dyn_strap_lower_jnt_03_capsule, center=<1.369 -0.0889 0.1304>, rotation=<rot 90 0 0>}]
 dyn_strap_lower_jnt_04 [{shapeType=Capsule, length=2.4926, radius=1, shapeName=dyn_strap_lower_jnt_04_capsule, center=<0.899 -0.0922 -0.0392>, rotation=<rot 90 0 0>}]
 dyn_strap_upper_jnt_01 [{shapeType=Capsule, length=1.8992, radius=1, shapeName=dyn_strap_upper_jnt_01_capsule, rotation=<rot 90 0 0>}]
 dyn_strap_upper_jnt_02 [{shapeType=Capsule, length=1.5776, radius=1, shapeName=dyn_strap_upper_jnt_02_capsule, rotation=<rot 90 0 0>}]
 for_pelvis_const_jnt [{
   shapeType=Capsule
   length=14.9978
   radius=10.8474
   shapeName=for_pelvis_const_jnt_capsule
   center=<-5.8428 -7.6901 7.2856>
   rotation=<rot 0 -90 19.445>
  }]
 pack [
  {
   shapeType=Box
   extentX=20.762
   extentY=8.9201
   extentZ=17.0756
   shapeName=pack_box
   center=<0.1233 -5.9126 1.4174>
   rotation=<rot -15.2892 0 -0.0001>
  }
  {
   shapeType=Box
   extentX=18.2335
   extentY=3.0612
   extentZ=26.6573
   shapeName=pack_box
   center=<-4.8972 -1.3114 -16.2773>
   rotation=<rot -14.4986 0.0001 -4.9115>
  }
  {
   shapeType=Box
   extentX=12.1475
   extentY=7.7587
   extentZ=7.5
   shapeName=pack_box
   center=<-2.0566 -7.3068 -6.1312>
   rotation=<rot -15.8008 0 0>
  }
 ]
 strap_upper_jnt_04 [{shapeType=Capsule, length=0.505, radius=0.505, shapeName=strap_upper_jnt_04_capsule}]
}
constraints [
 {bone1Name=dyn_strap_jnt_02, bone2Name=dyn_strap_jnt_01, swing1LimitDegrees=15, swing2LimitDegrees=22, twistLimitDegrees=5}
 {bone1Name=dyn_strap_jnt_03, bone2Name=dyn_strap_jnt_02, swing1LimitDegrees=15, swing2LimitDegrees=20, twistLimitDegrees=5}
 {bone1Name=dyn_strap_jnt_04, bone2Name=dyn_strap_jnt_03, swing1LimitDegrees=15, swing2LimitDegrees=18.5, twistLimitDegrees=5}
 {bone1Name=dyn_strap_jnt_05, bone2Name=dyn_strap_jnt_04, swing1LimitDegrees=15, swing2LimitDegrees=11, twistLimitDegrees=5}
 {
  bone1Name=dyn_strap_bind_jnt_01
  bone2Name=dyn_strap_jnt_05
  swing1LimitDegrees=5
  swing1Motion=Locked
  swing2LimitDegrees=3
  swing2Motion=Locked
  twistLimitDegrees=3
  twistMotion=Locked
 }
 {bone1Name=dyn_strap_lower_jnt_01, bone2Name=dyn_strap_bind_jnt_01, swing1LimitDegrees=15, swing2LimitDegrees=11, twistLimitDegrees=5}
 {bone1Name=dyn_strap_lower_jnt_02, bone2Name=dyn_strap_lower_jnt_01, swing1LimitDegrees=13, swing2LimitDegrees=15, twistLimitDegrees=5}
 {bone1Name=dyn_strap_lower_jnt_03, bone2Name=dyn_strap_lower_jnt_02, swing1LimitDegrees=15, swing2LimitDegrees=30, twistLimitDegrees=5}
 {bone1Name=dyn_strap_lower_jnt_04, bone2Name=dyn_strap_lower_jnt_03, swing1LimitDegrees=15, swing2LimitDegrees=40, twistLimitDegrees=5}
 {bone1Name=dyn_strap_upper_jnt_01, bone2Name=dyn_strap_bind_jnt_01, swing1LimitDegrees=2, swing2LimitDegrees=25, twistLimitDegrees=3}
 {bone1Name=dyn_strap_upper_jnt_02, bone2Name=dyn_strap_upper_jnt_01, swing1LimitDegrees=1, swing2LimitDegrees=10, twistLimitDegrees=1}
 {bone1Name=dyn_strap_jnt_01, bone2Name=pack, swing1LimitDegrees=7, swing2LimitDegrees=25, twistLimitDegrees=5}
 {bone1Name=strap_upper_jnt_04, bone2Name=dyn_strap_upper_jnt_02, swing1LimitDegrees=0.1, swing2LimitDegrees=10, twistLimitDegrees=1}
 {bone1Name=dyn_arrows_jnt_01, bone2Name=pack, swing1LimitDegrees=1, swing2LimitDegrees=5, twistLimitDegrees=1}
 {bone1Name=dyn_arrows_jnt_02, bone2Name=pack, swing1LimitDegrees=1, swing2LimitDegrees=5, twistLimitDegrees=1}
 {bone1Name=dyn_arrows_jnt_03, bone2Name=pack, swing1LimitDegrees=1, swing2LimitDegrees=4, twistLimitDegrees=1}
]
