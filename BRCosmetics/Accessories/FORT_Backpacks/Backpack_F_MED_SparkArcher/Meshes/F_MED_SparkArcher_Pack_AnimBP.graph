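# Anim Blueprint for the SparkArcher backpack cosmetic, posing the pack and simulating its secondary motion.
# AnimGraph offsets the pack bone for melee, parachute and skydive states, and pulls a constraint joint toward the owner's pelvis at partial alpha before a RigidBody node.
# EventGraph caches the pelvis world location and retunes the simulation's gravity and acceleration scale for front end, skydiving, parachuting or default play, using a deprecated gravity helper first.
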AnimBlueprint /BRCosmetics/Accessories/FORT_Backpacks/Backpack_F_MED_SparkArcher/Meshes/F_MED_SparkArcher_Pack_AnimBP
target_skeleton=&/F_MED_SparkArcher_Pack_Skeleton
graphs [
 AnimGraph {nodes [
   n1 = "Transform (Modify) Bone - Bone: pack"(ComponentPose=Local To Component(LocalPose=Input Pose()), Translation="0.000,0.219,0.000", Rotation="0,0,0") {
    translationMode=BMM_Additive
    translationSpace=BCS_BoneSpace
    alphaBoolBlend {blendOption=Linear}
    boneToModify {boneName=pack}
    translation {y=0.219}
   }
   n2 = "Transform (Modify) Bone - Bone: pack"!(ComponentPose=@n1, bAlphaBoolEnabled:bool=Get bIsPlayingMeleeAnim(), Translation="0,0,0", Rotation="0,-18,0", Scale) -> (Pose) {
    alphaInputType=Bool
    rotationMode=BMM_Additive
    rotationSpace=BCS_BoneSpace
    alphaBoolBlend {blendInTime=3, blendOption=CircularOut, blendOutTime=25}
    boneToModify {boneName=pack}
    rotation {yaw=-18}
   }
   n3 = "Transform (Modify) Bone - Bone: for_pelvis_const_jnt"(ComponentPose=@n2, Alpha=0.75, Translation=Get Location() -> (Location:Vector), Rotation="0,0,0", Scale="1,1,1") {
    translationMode=BMM_Replace
    translationSpace=BCS_WorldSpace
    alphaBoolBlend {blendOption=Linear}
    boneToModify {boneName=for_pelvis_const_jnt}
   }
   n4 = "Transform (Modify) Bone - Bone: pack"(ComponentPose=@n3, bAlphaBoolEnabled=Get bIsParachuteOpened(), Translation="0.000,-1.459,0.709", Rotation="-0.000,0.000,23.769") {
    alphaInputType=Bool
    rotationMode=BMM_Additive
    rotationSpace=BCS_BoneSpace
    translationMode=BMM_Additive
    translationSpace=BCS_BoneSpace
    alphaBoolBlend {blendInTime=0.3, blendOption=Linear, blendOutTime=0.3}
    boneToModify {boneName=pack}
    rotation {roll=23.769}
    translation {y=-1.459, z=0.709}
   }
   n5 = "Transform (Modify) Bone - Bone: pack"(ComponentPose=@n4, bAlphaBoolEnabled=Get bIsSkyDiving(), Translation="0.149,-3.894,-1.332", Rotation="-0.000,0.000,2.769") {
    alphaInputType=Bool
    rotationMode=BMM_Additive
    rotationSpace=BCS_BoneSpace
    translationMode=BMM_Additive
    translationSpace=BCS_BoneSpace
    alphaBoolBlend {blendOption=Linear}
    boneToModify {boneName=pack}
    rotation {roll=2.769}
    translation=<0.149 -3.894 -1.332>
   }
   n6 = RigidBody!(
    ComponentPose=@n5
    Alpha
    OverrideWorldGravity:Vector=Get Gravity Override()
    ExternalForce
    ComponentLinearAccScale:Vector=Get Component Linear Acc Scale() -> (Component Linear Acc Scale:Vector)
   ) -> (Pose) {
    bOverrideWorldGravity=true
    overlapChannel=ECC_PhysicsBody
    overridePhysicsAsset=&/F_MED_SparkArcher_Pack_Physics
    alphaBoolBlend {blendOption=Linear}
    componentAppliedLinearAccClamp=<1000 1000 1000>
    componentLinearAccScale=<2 0.1 0.3>
    componentLinearVelScale=<0.2 1.2 0.5>
    overrideWorldGravity {z=-650}
    simSpaceSettings {maxAngularAcceleration=1500, maxAngularVelocity=100, maxLinearAcceleration=1500, maxLinearVelocity=5}
   }
   Output Pose(Result=Component To Local(ComponentPose=@n6)) {layerGroup=None}
  ]}
 EventGraph {nodes [
   Event BlueprintUpdateAnimation()
   > n1 = Sequence!(execute) -> (then_0, then_1, then_2:Exec, then_3:Exec)
   > n2 = Gravity Override State Function DEPRECATED(CharacterPartAnimInstance=Self-Reference(), GameTimeInSeconds=GetGameTimeInSeconds(), GravityOverrideParams=Get Gravity Override Params())
   > Set Gravity Override(Gravity Override=@n2)
   n3 = Branch(execute=@n1.then_1, Condition=Get bIsInFrontEnd())
   > Set Component Linear Acc Scale(Component Linear Acc Scale:Vector="0.2,0,0.2") -> (Output_Get:Vector)
   n4 = Branch(execute=@n3.else, Condition=OR Boolean(A=Get bIsSkyDiving(), B=Get bIsParachuteOpened()))
   > n5 = Branch(Condition=Get bIsParachuteOpened())
   > Set Gravity Override(Gravity Override="0,0,-100")
   > Set Component Linear Acc Scale(Component Linear Acc Scale="0.35,0.2,0.05")
   n6 = vector + vector(B="0,0,1000")
   n7 = Set Gravity Override(execute=@n4.else, Gravity Override="0,0,-650")
   n8 = vector + vector(A=@n7, B="0,0,-200")
   Set Gravity Override(
    execute=@n5.else
    Gravity Override=TInterpTo!(
     Current:"Transform (by ref)"="To Transform (Vector)"!(InLocation:Vector=@n8) -> (ReturnValue:Transform)
     Target:"Transform (by ref)"="To Transform (Vector)"(InLocation=@n6)
     DeltaTime:float=4
     InterpSpeed:float=0.1
    ) -> (ReturnValue_Location:Vector, ReturnValue_Rotation:Rotator, ReturnValue_Scale:Vector)
   )
   Set Location(
    execute=@n1.then_3
    Location:Vector=GetSocketTransform!(self=GetOwnerSkeletalMeshComponent(), InSocketName=pelvis, TransformSpace=RTS_World) -> (ReturnValue_Location:Vector, ReturnValue_Rotation:Rotator, ReturnValue_Scale:Vector)
   ) -> (Output_Get:Vector)
   TryGetPawnOwner()
   Set Component Linear Acc Scale(execute=@n7.then, Component Linear Acc Scale="2,0.1,0.3")
  ]}
]
variables [
 Component Linear Acc Scale {type=Vector, default=<2 0.1 0.3>}
 Gravity Override {type=Vector}
 Gravity Override Params {type=GravityOverrideParamsStruct}
 Location {type=Vector}
]
